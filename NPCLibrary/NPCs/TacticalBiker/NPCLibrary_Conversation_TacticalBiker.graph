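# Conversation database for the Tactical Biker NPC, defining the greeting shown when a player starts talking to it.
# A single speech node gives a default line plus contextual alternatives for a first-ever conversation or specific player outfits, ranked by curve-driven priorities.
# The speech then branches through link nodes to the shared NPC quests and basic services entries and to the biker's own dialogue entry.
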ConversationDatabase /NPCLibrary/NPCs/TacticalBiker/NPCLibrary_Conversation_TacticalBiker
entry_points [{tag=Conversation.Entry.TacticalBiker, destinations [#1]}]
nodes [
 ConversationEntryPointNode_7 {
  id=#1
  class=ConversationEntryPointNode
  outputs [#2]
  properties {entryTag {tagName=Conversation.Entry.TacticalBiker}}
 }
 Speech_C_4 {
  id=#2
  class=Speech_C
  outputs [#3, #4, #5]
  properties {
   nodeName=""
   generalConfig {
    defaultMessage=Better hurry. I won't be here long.
    contextualMessages [
     {
      message="If you think you're going too fast, you're already behind."
      contextRequirements [{requirement {class=FortControllerRequirement_IsFirstEverConversationWithNPC}}]
      priority {curve {rowName=Tandem.Speech.Priority.FirstEverInteraction}}
     }
     {
      message=Another racer? Let's see what you've got.
      contextRequirements [{requirement {class=FortControllerRequirement_HasCID, properties {allowedCIds [/BRCosmetics/Athena/Items/Cosmetics/Characters/Character_RankedSpeeder]}}}]
      priority {curve {rowName=Tandem.Speech.Priority.OutfitReaction}}
     }
     {
      message="It's good to see you, Burnout."
      contextRequirements [{requirement {
         class=FortControllerRequirement_HasCID
         properties {allowedCIds [/BRCosmetics/Athena/Items/Cosmetics/Characters/CID_094_Athena_Commando_M_Rider]}
        }}]
      priority {curve {rowName=Tandem.Speech.Priority.OutfitReaction}}
     }
     {
      message="Ready for a rematch, Redline?"
      contextRequirements [{requirement {
         class=FortControllerRequirement_HasCID
         properties {allowedCIds [/BRCosmetics/Athena/Items/Cosmetics/Characters/CID_162_Athena_Commando_F_StreetRacer]}
        }}]
      priority {curve {rowName=Tandem.Speech.Priority.OutfitReaction}}
     }
     {
      message="You know, you're the only one I've ever lost a race to."
      contextRequirements [{requirement {
         class=FortControllerRequirement_HasCID
         properties {allowedCIds [/BRCosmetics/Athena/Items/Cosmetics/Characters/CID_524_Athena_Commando_F_FutureBikerWhite]}
        }}]
      priority {curve {rowName=Tandem.Speech.Priority.OutfitReaction}}
     }
    ]
   }
  }
 }
 ConversationLinkNode_2 {
  id=#5
  class=ConversationLinkNode
  properties {remoteEntryTag {tagName=Conversation.Entry.TacticalBiker.Dialogue}}
 }
 ConversationLinkNode_0 {id=#4, class=ConversationLinkNode, properties {remoteEntryTag {tagName=Conversation.Entry.BasicServices}}}
 ConversationLinkNode_1 {id=#3, class=ConversationLinkNode, properties {remoteEntryTag {tagName=Conversation.Entry.NPCLibrary.Quests}}}
]
