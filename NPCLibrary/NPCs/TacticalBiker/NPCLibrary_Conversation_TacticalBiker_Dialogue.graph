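# Conversation database for the Tactical Biker NPC's lore dialogue, entered through the Conversation.Entry.TacticalBiker.Dialogue tag.
# An "Ask about..." speech node greets the player and branches into RPM, Drakon Steel Rider, or a Back choice.
# Each topic plays a linear chain of speech lines about the biker's racing squad or that rider.
# Both chains converge on one shared ReturnToStart node shown as BACK.
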ConversationDatabase /NPCLibrary/NPCs/TacticalBiker/NPCLibrary_Conversation_TacticalBiker_Dialogue
entry_points [{tag=Conversation.Entry.TacticalBiker.Dialogue, destinations [#1]}]
nodes [
 ConversationEntryPointNode_0 {
  id=#1
  class=ConversationEntryPointNode
  outputs [#2]
  properties {entryTag {tagName=Conversation.Entry.TacticalBiker.Dialogue}}
 }
 Speech_C_0 {
  id=#2
  class=Speech_C
  outputs [#3, #4, #5]
  properties {
   nodeName=""
   softTaskIcon=&T_UI_NPC_question
   generalConfig {defaultMessage=I've got some time before my next race. What's up?}
  }
  sub_nodes [ConversationChoiceNode_5 {class=ConversationChoiceNode, properties {defaultChoiceDisplayText=Ask about...}}]
 }
 Speech_C_2 {
  id=#5
  class=Speech_C
  outputs [#6]
  properties {
   nodeName=""
   softTaskIcon=&T_UI_NPC_question
   generalConfig {defaultMessage="The RPM squad and I go way back. Redline, Burnout, even Payback."}
  }
  sub_nodes [ConversationChoiceNode_15 {class=ConversationChoiceNode, properties {defaultChoiceDisplayText=RPM}}]
 }
 Speech_C_3 {
  id=#6
  class=Speech_C
  outputs [#7]
  properties {
   nodeName=""
   softTaskIcon=&T_UI_NPC_question
   generalConfig {defaultMessage=Best racers I've ever competed against. Good friends.}
  }
 }
 Speech_C_4 {
  id=#7
  class=Speech_C
  outputs [#8]
  properties {
   nodeName=""
   softTaskIcon=&T_UI_NPC_question
   generalConfig {defaultMessage="Still, it's always felt like we're missing a racer in our lineup, you know?"}
  }
 }
 ReturnToStart_C_23 {
  id=#8
  class=ReturnToStart_C
  properties {softTaskIcon=&T_UI_NPC_Back}
  sub_nodes [ConversationChoiceNode_32 {class=ConversationChoiceNode, properties {defaultChoiceDisplayText=BACK}}]
 }
 Back_C_0 {
  id=#4
  class=Back_C
  sub_nodes [ConversationChoiceNode_7 {class=ConversationChoiceNode, properties {defaultChoiceDisplayText=BACK}}]
 }
 Speech_C_7 {
  id=#3
  class=Speech_C
  outputs [#9]
  properties {
   nodeName=""
   softTaskIcon=&T_UI_NPC_question
   generalConfig {defaultMessage="Yeah, I've heard of Drakon Steel Rider. Even met him once."}
  }
  sub_nodes [ConversationChoiceNode_0 {class=ConversationChoiceNode, properties {defaultChoiceDisplayText=Drakon Steel Rider}}]
 }
 Speech_C_15 {
  id=#9
  class=Speech_C
  outputs [#10]
  properties {
   nodeName=""
   softTaskIcon=&T_UI_NPC_question
   generalConfig {defaultMessage="He's a good racer, but...a little full of himself."}
  }
 }
 Speech_C_14 {
  id=#10
  class=Speech_C
  outputs [#11]
  properties {
   nodeName=""
   softTaskIcon=&T_UI_NPC_question
   generalConfig {defaultMessage="Maybe there could be a place for him on RPM, though."}
  }
 }
 Speech_C_1 {
  id=#11
  class=Speech_C
  outputs [#8]
  properties {nodeName="", softTaskIcon=&T_UI_NPC_question, generalConfig {defaultMessage=If he asked.}}
 }
]
